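# Sound cue asset for medium predator vocalizations, filed among the shared sound indicator assets.
# It has no root or sound nodes, so it produces no audio itself and carries only playback settings.
# These pair medium predator vocalization attenuation and the creature vocalization sound class with high priority, boosted volume and a set max distance.
# Owner and global concurrency limits are borrowed from Robert's vocal breathing sounds.
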SoundCue /Irwin/Sounds/Shared/IndicatorAssets/SoundIndicator_Predator_Vocalization_Medium_Cue
root=null
nodes []
properties {
 attenuationSettings=/Irwin/Sounds/Settings/Predator_Vocalization_Medium_Att
 maxDistance=12000
 priority=0.95
 soundClassObject=/Game/Sounds/FortSoundClassesAndModes/Fort_Creatures_Vocalization_Normal
 volumeMultiplier=1.25
 concurrencySet [/Irwin/Sounds/Robert/Robert_Vocal_Breathing_Concurrency_Owner, /Irwin/Sounds/Robert/Robert_Vocal_Breathing_Concurrency_Global]
}
